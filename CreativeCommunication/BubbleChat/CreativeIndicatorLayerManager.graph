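# Widget Blueprint layer manager for Creative communication indicators in the BubbleChat folder, derived from the native CreativeIndicatorLayerManagerWidget.
# An Overlay root holds the inherited PlayerIndicators FortActorCanvas, stretched to fill, drawing elements in order with direction and elevation arrow brushes.
# A TutorialHighlight widget is also listed without a parent.
# The event graph has only an unconnected Construct event, so the Blueprint supplies layout and styling and adds no logic.
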WidgetBlueprint /CreativeCommunication/BubbleChat/CreativeIndicatorLayerManager
graphs [EventGraph {nodes [Event Construct()]}]
variables []
widgets {
 info {
  inheritedWidgetCount=2
  parentClass=/Script/CreativeCommunicationUI.CreativeIndicatorLayerManagerWidget
  rootWidgetClass=&Overlay
  widgetCount=3
 }
 widgets [
  Overlay_0 {widgetClassPath=&Overlay, widget=Overlay_0}
  PlayerIndicators {
   widgetClassPath=/Script/FortniteUI.FortActorCanvas
   bInherited=true
   bIsVariable=true
   parent=Overlay_0
   widget=PlayerIndicators
   properties {
    bDrawElementsInOrder=true
    arrowIconBrush {
     bIsDynamicallyLoaded=false
     drawAs=Image
     imageType=FullColor
     mirroring=NoMirror
     resourceName=None
     resourceObject=/Game/UI/CodeReferenced/T_UI_DirectionArrow
     tiling=NoTile
     imageSize=<32 32>
     margin=<ltrb 0 0 0 0>
     outlineSettings {
      bUseBrushTransparency=false
      roundingType=HalfHeightRadius
      width=0
      color {colorUseRule=UseColor_Specified, specifiedColor=<rgba 0 0 0 0>}
      cornerRadii=<0 0 0 0>
     }
     tintColor {colorUseRule=UseColor_Specified, specifiedColor=<rgba 1 1 1 1>}
     uVRegion {bIsValid=false, max=<0 0>, min=<0 0>}
    }
    elevationArrowIconBrush {
     bIsDynamicallyLoaded=false
     drawAs=Image
     imageType=FullColor
     mirroring=NoMirror
     resourceName=None
     resourceObject=/Game/UI/CodeReferenced/T_UI_ElevationArrow
     tiling=NoTile
     imageSize=<32 32>
     margin=<ltrb 0 0 0 0>
     outlineSettings {
      bUseBrushTransparency=false
      roundingType=HalfHeightRadius
      width=0
      color {colorUseRule=UseColor_Specified, specifiedColor=<rgba 0 0 0 0>}
      cornerRadii=<0 0 0 0>
     }
     tintColor {colorUseRule=UseColor_Specified, specifiedColor=<rgba 1 1 1 1>}
     uVRegion {bIsValid=false, max=<0 0>, min=<0 0>}
    }
   }
   slot {class=OverlaySlot, properties {horizontalAlignment=HAlign_Fill, verticalAlignment=VAlign_Fill}}
  }
  TutorialHighlight {widgetClassPath=&TutorialHighlight}
 ]
}
