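# WBP_CreativeBubbleChatEntry is a list entry widget that displays a single text message in Creative bubble chat.
# When a list item is set, it casts the item to FortTextChatMessageVM and assigns it as the widget's viewmodel.
# SetUserMessage centers text of five or fewer characters and left-aligns longer text before setting it.
# A clipped, size-constrained SizeBox holds the text block, which wraps per character and truncates with a multiline ellipsis.
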WidgetBlueprint /CreativeCommunication/BubbleChat/WBP_CreativeBubbleChatEntry
graphs [
 EventGraph {nodes [
   n1 = Event OnListItemObjectSet()
   > n2 = Cast To FortTextChatMessageVM!(execute, Object=@n1.ListItemObject) -> (then, CastFailed, AsText Chat Message VM:Text Chat Message VM Object Reference) {is=title K2Node_DynamicCast}
   > SetFortTextChatMessageVM!(execute, self, Viewmodel:Text Chat Message VM Object Reference=@n2) -> (then)
  ]}
 SetUserMessage {nodes [
   n1 = SetUserMessage!() -> (then, Text:Text) {is=title K2Node_FunctionEntry}
   n2 = GetCharacterArrayFromString!(SourceString:String="To String (Text)"(InText=@n1)) -> (ReturnValue:Array of Strings)
   > n3 = Branch(Condition="integer <= integer"(A=Length(TargetArray:Array of Strings=@n2), B=5))
   n4 = SlotAsSizeBoxSlot!(Widget:Widget Object Reference=Get UserMessage() -> (UserMessage:Common Text Object Reference)) -> (ReturnValue:Size Box Slot Object Reference)
   > n5 = SetHorizontalAlignment(self:Size Box Slot Object Reference=@n4, InHorizontalAlignment=HAlign_Center)
   "SetText (Text)"(execute=[@n6.then, @n5.then], self=Get UserMessage(), InText=@n1)
   n6 = SetHorizontalAlignment(execute=@n3.else, self=@n4, InHorizontalAlignment=HAlign_Left)
  ]}
]
variables []
widgets {
 info {inheritedWidgetCount=0, parentClass=&UserWidget, rootWidgetClass=&SizeBox, widgetCount=2}
 widgets [
  WrappingEnforcingSB {
   widgetClassPath=&SizeBox
   widget=WrappingEnforcingSB
   properties {
    bOverride_MaxDesiredHeight=true
    bOverride_MaxDesiredWidth=true
    bOverride_MinDesiredHeight=true
    bOverride_MinDesiredWidth=true
    clipping=ClipToBounds
    maxDesiredHeight=80
    maxDesiredWidth=300
    minDesiredHeight=32
    minDesiredWidth=64
   }
  }
  UserMessage {
   widgetClassPath=&CommonTextBlock
   bIsVariable=true
   parent=WrappingEnforcingSB
   widget=UserMessage
   properties {
    applyLineHeightToBottomLine=false
    bIsScrollingEnabled=false
    mobileFontSizeMultiplier=1.2
    text=Very long user message that can wrap to up to three lines vertically. This text is to test that scenario in design view.
    textOverflowPolicy=MultilineEllipsis
    wrapTextAt=300
    wrappingPolicy=AllowPerCharacterWrapping
    font {
     bForceMonospaced=false
     bMaterialIsStencil=false
     fontMaterial=null
     fontObject=&HeadingNow
     letterSpacing=0
     monospacedWidth=1
     size=15.75
     skewAmount=0
     typefaceFontName=75 Medium
     outlineSettings {
      bApplyOutlineToDropShadows=false
      bMiteredCorners=false
      bSeparateFillAlpha=false
      outlineMaterial=null
      outlineSize=0
      outlineColor=<rgba 0 0 0 1>
     }
    }
    shadowOffset=<0 0>
   }
   slot {class=SizeBoxSlot, properties {horizontalAlignment=HAlign_Left, verticalAlignment=VAlign_Center}}
  }
 ]
}
