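# Chat speech bubble widget that shows the newest message over one of ten switchable background styles, from Default and Anime to Neon and Retro.
# FillOutTextChatData clears the list, toggles background visibility, plays the pop-in Intro, then lists only the last message view model.
# SetVisualStyle tints only the Anime image for whispers, then calls SetTextColor.
# SetTextColor, also run on PreConstruct, colors the list's parent border content by active switcher index.
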WidgetBlueprint /CreativeCommunication/BubbleChat/WBP_CreativeBubbleChat_V2
animations [Intro {
  binding_count=3
  length_seconds=0.4569
  object=&self:Intro
  sequence {bindings [
    Overlay_BubbleContainer {
     id=#1
     possessed_class=Overlay
     tracks [
      {
       type=MovieSceneFloatTrack
       display_name=Render Opacity
       sections [{type=MovieSceneFloatSection, channels [None {type=MovieSceneScriptingFloatChannel, keys [[0.039,0],[3.0753,0],[6.9057,1]]}]}]
      }
      {
       type=MovieScene2DTransformTrack
       sections [{
         type=MovieScene2DTransformSection
         channels [
          Translation.X {type=MovieSceneScriptingFloatChannel}
          Translation.Y {type=MovieSceneScriptingFloatChannel, keys [[3.0753,15],[6.9057,0]]}
          Angle {type=MovieSceneScriptingFloatChannel}
          Scale.X {type=MovieSceneScriptingFloatChannel, keys [[3.0753,0.25],[5.4153,1.05],[6.3423,0.98],[6.9057,1]]}
          Scale.Y {type=MovieSceneScriptingFloatChannel, keys [[3.0753,0.25],[5.4153,1.05],[6.3423,0.98],[6.9057,1]]}
          Shear.X {type=MovieSceneScriptingFloatChannel}
          Shear.Y {type=MovieSceneScriptingFloatChannel}
         ]
        }]
      }
     ]
    }
    Anime {
     id=#2
     possessed_class=Image
     tracks [{
       type=MovieSceneWidgetMaterialTrack
       sections [{type=MovieSceneParameterSection, channels [TailScale {type=MovieSceneScriptingFloatChannel, keys [[4.2473,0.5],[9.0997,1]]}]}]
      }]
    }
    ListView_Messages {
     id=#3
     possessed_class=CommonListView
     tracks [{
       type=MovieSceneFloatTrack
       display_name=Render Opacity
       sections [{type=MovieSceneFloatSection, channels [None {type=MovieSceneScriptingFloatChannel, keys [[5.2047,0],[9.177,1]]}]}]
      }]
    }
   ]}
 }]
graphs [
 EventGraph {nodes [
   Event PreConstruct()
   > SetTextColor()
   Event OnInitialized()
   > SetScrollbarVisibility!(
    execute
    self:List View Base Object Reference=Get ListView_Messages() -> (ListView_Messages:Common List View Object Reference)
    InVisibility:ESlateVisibility Enum=Collapsed
   ) -> (then)
  ]}
 FillOutTextChatData {nodes [
   FillOutTextChatData!() -> (then, Messages:Array of Creative Bubble Chat Message View Model Object References) {is=title K2Node_FunctionEntry}
   > ClearListItems(self=Get ListView_Messages())
   n1 = Is Not Empty!(TargetArray:Array of Creative Bubble Chat Message View Model Object References=Get Messages() -> (Messages:Array of Creative Bubble Chat Message View Model Object References)) -> (ReturnValue:bool)
   > SetVisibility(
    self=Get Switcher_Styles() -> (Switcher_Styles:Widget Switcher Object Reference)
    InVisibility=Select(Option 0=Collapsed, Option 1=SelfHitTestInvisible, Index=@n1)
   )
   > PlayAnimation(InAnimation=Get Intro(), PlaybackSpeed=0.75)
   > Branch(Condition=@n1)
   n2 = Get Messages()
   n3 = "Get (a copy)"(
    Array:Array of Creative Bubble Chat Message View Model Object References=@n2
    Dimension 1=Last Index!(TargetArray:Array of Creative Bubble Chat Message View Model Object References=@n2) -> (ReturnValue=0)
   ) -> (Output:Creative Bubble Chat Message View Model Object Reference)
   > SetVisualStyle!(execute, self, IsWhisper:bool=Get IsWhisper!(self:Creative Bubble Chat Message View Model Object Reference=@n3) -> (IsWhisper:bool)) -> (then)
   > Set List Items!(
    execute
    self:List View Object Reference=Get ListView_Messages()
    InListItems:Array of Object References=Make Array("[0]":Object Reference=@n3) -> (Array:Array of Object References)
   ) -> (then) {is=title K2Node_CallFunction}
   > ScrollToBottom(self:List View Base Object Reference=Get ListView_Messages())
  ]}
 SetTextColor {nodes [
   SetTextColor() {is=title K2Node_FunctionEntry}
   > n1 = Switch on Int!(execute:Exec, Selection:int=GetActiveWidgetIndex!(self:Widget Switcher Object Reference=Get Switcher_Styles()) -> (ReturnValue:int=0)) -> (0:Exec, 1:Exec, 2:Exec, 3:Exec, 4:Exec, 5:Exec, 6:Exec, 7:Exec, 8:Exec, 9:Exec) {bHasDefaultPin=false}
   > SetContentColorAndOpacity!(
    execute
    self:Border Object Reference=Get ColorSwitchingCB() -> (ColorSwitchingCB:Common Border Object Reference)
    InContentColorAndOpacity:Linear Color Structure=(R=1,G=1,B=1,A=1)
   ) -> (then)
   SetContentColorAndOpacity(execute=@n1.1, self=Get ColorSwitchingCB(), InContentColorAndOpacity=(R=0,G=0,B=0,A=1))
   SetContentColorAndOpacity(execute=@n1.2, self=Get ColorSwitchingCB(), InContentColorAndOpacity=(R=0.0262,G=0.0262,B=0.0262,A=1))
   SetContentColorAndOpacity(execute=@n1.3, self=Get ColorSwitchingCB(), InContentColorAndOpacity=(R=0.7305,G=0.5647,B=0.1356,A=1))
   SetContentColorAndOpacity(execute=@n1.4, self=Get ColorSwitchingCB(), InContentColorAndOpacity=(R=1,G=1,B=1,A=1))
   SetContentColorAndOpacity(execute=@n1.5, self=Get ColorSwitchingCB(), InContentColorAndOpacity=(R=0.0666,G=0.8879,B=1,A=1))
   SetContentColorAndOpacity(execute=@n1.6, self=Get ColorSwitchingCB(), InContentColorAndOpacity=(R=0.107,G=0.107,B=0.107,A=1))
   SetContentColorAndOpacity(execute=@n1.7, self=Get ColorSwitchingCB(), InContentColorAndOpacity=(R=0,G=0,B=0,A=1))
   SetContentColorAndOpacity(execute=@n1.8, self=Get ColorSwitchingCB(), InContentColorAndOpacity=(R=1,G=1,B=1,A=1))
   SetContentColorAndOpacity(execute=@n1.9, self=Get ColorSwitchingCB(), InContentColorAndOpacity=(R=1,G=1,B=1,A=0.8))
  ]}
 SetVisualStyle {nodes [
   n1 = SetVisualStyle!() -> (then, IsWhisper:bool=~) {is=title K2Node_FunctionEntry}
   > SetColorAndOpacity(
    self:Image Object Reference=Get Anime() -> (Anime:Image Object Reference)
    InColorAndOpacity:Linear Color Structure=Select(Option 0:Linear Color Structure=(R=0.624,G=0.7305,B=0.8879,A=1), Option 1:Linear Color Structure=(R=0.1413,G=0.1912,B=0.8714,A=1), Index=@n1) -> (ReturnValue:Linear Color Structure)
   )
   > SetTextColor() {is=title K2Node_CallFunction}
  ]}
]
variables []
widgets {
 info {inheritedWidgetCount=0, parentClass=&UserWidget, rootWidgetClass=&Overlay, widgetCount=19}
 widgets [
  Overlay_BubbleContainer {widgetClassPath=&Overlay, widget=Overlay_BubbleContainer}
  Switcher_Styles {
   widgetClassPath=&WidgetSwitcher
   bIsVariable=true
   parent=Overlay_BubbleContainer
   widget=Switcher_Styles
   slot {class=OverlaySlot, properties {horizontalAlignment=HAlign_Fill, verticalAlignment=VAlign_Fill}}
  }
  Default {
   widgetClassPath=&Image
   bIsVariable=true
   parent=Switcher_Styles
   widget=Default
   properties {
    visibility=SelfHitTestInvisible
    brush {
     bIsDynamicallyLoaded=false
     drawAs=Box
     imageType=FullColor
     mirroring=NoMirror
     resourceName=None
     resourceObject=/CreativeCommunication/BubbleChat/Art/MI_UI_Creative_SpeechBubble_OverheadDialogue_DefaultStyle
     tiling=NoTile
     imageSize=<94 94>
     margin=<ltrb 0.5 0.5 0.5 0.5>
     outlineSettings {
      bUseBrushTransparency=false
      roundingType=HalfHeightRadius
      width=0
      color {colorUseRule=UseColor_Specified, specifiedColor=<rgba 0 0 0 0>}
      cornerRadii=<0 0 0 0>
     }
     tintColor {colorUseRule=UseColor_Specified, specifiedColor=<rgba 1 1 1 1>}
     uVRegion {bIsValid=false, max=<0 0>, min=<0 0>}
    }
    colorAndOpacity=<rgba 0 0.0976 0.521 1>
    renderTransformPivot=<0.5 0.9>
   }
   slot {class=WidgetSwitcherSlot, properties {padding=<ltrb -23 -16 -23 -23>}}
  }
  Anime {
   widgetClassPath=&Image
   bIsVariable=true
   parent=Switcher_Styles
   widget=Anime
   properties {
    visibility=SelfHitTestInvisible
    brush {
     bIsDynamicallyLoaded=false
     drawAs=Box
     imageType=FullColor
     mirroring=NoMirror
     resourceName=None
     resourceObject=/CreativeCommunication/BubbleChat/Art/Textures/T_UI_BubbleChatStyle_Anime
     tiling=NoTile
     imageSize=<32 32>
     margin=<ltrb 0 0 0 0>
     outlineSettings {
      bUseBrushTransparency=false
      roundingType=HalfHeightRadius
      width=0
      color {colorUseRule=UseColor_Specified, specifiedColor=<rgba 0 0 0 0>}
      cornerRadii=<0 0 0 0>
     }
     tintColor {colorUseRule=UseColor_Specified, specifiedColor=<rgba 1 1 1 1>}
     uVRegion {bIsValid=false, max=<0 0>, min=<0 0>}
    }
    renderTransformPivot=<0.5 0.9>
   }
   slot {class=WidgetSwitcherSlot, properties {padding=<ltrb -23 -12 -23 -40>}}
  }
  Comic {
   widgetClassPath=&Image
   bIsVariable=true
   parent=Switcher_Styles
   widget=Comic
   properties {
    visibility=SelfHitTestInvisible
    brush {
     bIsDynamicallyLoaded=false
     drawAs=Box
     imageType=FullColor
     mirroring=NoMirror
     resourceName=None
     resourceObject=/CreativeCommunication/BubbleChat/Art/Textures/T_UI_BubbleChatStyle_Comic
     tiling=NoTile
     imageSize=<32 32>
     margin=<ltrb 0.5 0.4685 0.5 0.5277>
     outlineSettings {
      bUseBrushTransparency=false
      roundingType=HalfHeightRadius
      width=0
      color {colorUseRule=UseColor_Specified, specifiedColor=<rgba 0 0 0 0>}
      cornerRadii=<0 0 0 0>
     }
     tintColor {colorUseRule=UseColor_Specified, specifiedColor=<rgba 1 1 1 1>}
     uVRegion {bIsValid=false, max=<0 0>, min=<0 0>}
    }
    renderTransformPivot=<0.5 0.9>
   }
   slot {class=WidgetSwitcherSlot, properties {padding=<ltrb -23 0 -23 -40>}}
  }
  Fantasy {widgetClassPath=&Overlay, parent=Switcher_Styles, widget=Fantasy, slot {class=WidgetSwitcherSlot}}
  Fantasy_Bubble {
   widgetClassPath=&Image
   bIsVariable=true
   parent=Fantasy
   widget=Fantasy_Bubble
   properties {
    visibility=SelfHitTestInvisible
    brush {
     bIsDynamicallyLoaded=false
     drawAs=Box
     imageType=FullColor
     mirroring=NoMirror
     resourceName=None
     resourceObject=/CreativeCommunication/BubbleChat/Art/Textures/T_UI_BubbleChatStyle_Fantasy
     tiling=NoTile
     imageSize=<94 94>
     margin=<ltrb 0.2856 0.2416 0.3296 0.5672>
     outlineSettings {
      bUseBrushTransparency=false
      roundingType=HalfHeightRadius
      width=0
      color {colorUseRule=UseColor_Specified, specifiedColor=<rgba 0 0 0 0>}
      cornerRadii=<0 0 0 0>
     }
     tintColor {colorUseRule=UseColor_Specified, specifiedColor=<rgba 1 1 1 1>}
     uVRegion {bIsValid=false, max=<0 0>, min=<0 0>}
    }
    renderTransformPivot=<0.5 0.9>
   }
   slot {class=OverlaySlot, properties {horizontalAlignment=HAlign_Fill, verticalAlignment=VAlign_Fill, padding=<ltrb -23 -4 -23 -34>}}
  }
  Decoration_L {
   widgetClassPath=&Image
   bIsVariable=true
   parent=Fantasy
   widget=Decoration_L
   properties {
    brush {
     bIsDynamicallyLoaded=false
     drawAs=Image
     imageType=FullColor
     mirroring=NoMirror
     resourceName=None
     resourceObject=/CreativeCommunication/BubbleChat/Art/Textures/T_UI_BubbleChatStyle_Fantasy_Ornament
     tiling=NoTile
     imageSize=<26 38>
     margin=<ltrb 0 0 0 0>
     outlineSettings {
      bUseBrushTransparency=false
      roundingType=HalfHeightRadius
      width=0
      color {colorUseRule=UseColor_Specified, specifiedColor=<rgba 0 0 0 0>}
      cornerRadii=<0 0 0 0>
     }
     tintColor {colorUseRule=UseColor_Specified, specifiedColor=<rgba 1 1 1 1>}
     uVRegion {bIsValid=false, max=<0 0>, min=<0 0>}
    }
    renderTransform {angle=180, scale=<1 -1>, shear=<0 0>, translation=<0 0>}
   }
   slot {class=OverlaySlot, properties {verticalAlignment=VAlign_Center, padding=<ltrb -34 0 0 -2>}}
  }
  Decoration_R {
   widgetClassPath=&Image
   bIsVariable=true
   parent=Fantasy
   widget=Decoration_R
   properties {brush {
     bIsDynamicallyLoaded=false
     drawAs=Image
     imageType=FullColor
     mirroring=NoMirror
     resourceName=None
     resourceObject=&T_UI_BubbleChatStyle_Fantasy_Ornament
     tiling=NoTile
     imageSize=<26 38>
     margin=<ltrb 0 0 0 0>
     outlineSettings {
      bUseBrushTransparency=false
      roundingType=HalfHeightRadius
      width=0
      color {colorUseRule=UseColor_Specified, specifiedColor=<rgba 0 0 0 0>}
      cornerRadii=<0 0 0 0>
     }
     tintColor {colorUseRule=UseColor_Specified, specifiedColor=<rgba 1 1 1 1>}
     uVRegion {bIsValid=false, max=<0 0>, min=<0 0>}
    }}
   slot {class=OverlaySlot, properties {horizontalAlignment=HAlign_Right, verticalAlignment=VAlign_Center, padding=<ltrb 0 0 -34 -2>}}
  }
  Gothic {
   widgetClassPath=&Image
   bIsVariable=true
   parent=Switcher_Styles
   widget=Gothic
   properties {
    visibility=SelfHitTestInvisible
    brush {
     bIsDynamicallyLoaded=false
     drawAs=Box
     imageType=FullColor
     mirroring=NoMirror
     resourceName=None
     resourceObject=/CreativeCommunication/BubbleChat/Art/Textures/T_UI_BubbleChatStyle_Gothic
     tiling=NoTile
     imageSize=<32 32>
     margin=<ltrb 0.49 0.49 0.49 0.49>
     outlineSettings {
      bUseBrushTransparency=false
      roundingType=HalfHeightRadius
      width=0
      color {colorUseRule=UseColor_Specified, specifiedColor=<rgba 0 0 0 0>}
      cornerRadii=<0 0 0 0>
     }
     tintColor {colorUseRule=UseColor_Specified, specifiedColor=<rgba 1 1 1 1>}
     uVRegion {bIsValid=false, max=<0 0>, min=<0 0>}
    }
    colorAndOpacity=<rgba 0 0 0 1>
    renderTransformPivot=<0.5 0.9>
   }
   slot {class=WidgetSwitcherSlot, properties {padding=<ltrb -23 -16 -23 -36>}}
  }
  Neon {widgetClassPath=&Overlay, parent=Switcher_Styles, widget=Neon, slot {class=WidgetSwitcherSlot}}
  Neon_Bubble {
   widgetClassPath=&Image
   bIsVariable=true
   parent=Neon
   widget=Neon_Bubble
   properties {
    visibility=SelfHitTestInvisible
    brush {
     bIsDynamicallyLoaded=false
     drawAs=RoundedBox
     imageType=FullColor
     mirroring=NoMirror
     resourceName=None
     resourceObject=null
     tiling=NoTile
     imageSize=<32 32>
     margin=<ltrb 0.35 0.5 0.35 0.503>
     outlineSettings {
      bUseBrushTransparency=false
      roundingType=FixedRadius
      width=3
      color {colorUseRule=UseColor_Specified, specifiedColor=<rgba 0.0723 0.7231 0.9647 1>}
      cornerRadii=<12 12 12 12>
     }
     tintColor {colorUseRule=UseColor_Specified, specifiedColor=<rgba 0.0296 0.0152 0.0482 1>}
     uVRegion {bIsValid=false, max=<0 0>, min=<0 0>}
    }
    renderTransformPivot=<0.5 0.9>
   }
   slot {class=OverlaySlot, properties {horizontalAlignment=HAlign_Fill, verticalAlignment=VAlign_Fill}}
  }
  Neon_Tail {
   widgetClassPath=&Image
   bIsVariable=true
   parent=Neon
   widget=Neon_Tail
   properties {
    visibility=SelfHitTestInvisible
    brush {
     bIsDynamicallyLoaded=false
     drawAs=Image
     imageType=FullColor
     mirroring=NoMirror
     resourceName=None
     resourceObject=/CreativeCommunication/BubbleChat/Art/Textures/T_UI_BubbleChatStyle_Neon_Tail
     tiling=NoTile
     imageSize=<51 48>
     margin=<ltrb 0.35 0.5 0.35 0.503>
     outlineSettings {
      bUseBrushTransparency=false
      roundingType=FixedRadius
      width=3
      color {colorUseRule=UseColor_Specified, specifiedColor=<rgba 0.0723 0.7231 0.9647 1>}
      cornerRadii=<12 12 12 12>
     }
     tintColor {colorUseRule=UseColor_Specified, specifiedColor=<rgba 1 1 1 1>}
     uVRegion {bIsValid=false, max=<0 0>, min=<0 0>}
    }
    renderTransformPivot=<0.5 0.9>
   }
   slot {class=OverlaySlot, properties {horizontalAlignment=HAlign_Center, verticalAlignment=VAlign_Bottom, padding=<ltrb 10 0 0 -44.7>}}
  }
  Pixelated {
   widgetClassPath=&Image
   bIsVariable=true
   parent=Switcher_Styles
   widget=Pixelated
   properties {
    visibility=SelfHitTestInvisible
    brush {
     bIsDynamicallyLoaded=false
     drawAs=Box
     imageType=FullColor
     mirroring=NoMirror
     resourceName=None
     resourceObject=/CreativeCommunication/BubbleChat/Art/Textures/T_UI_BubbleChatStyle_Pixelated
     tiling=NoTile
     imageSize=<94 94>
     margin=<ltrb 0.35 0.17 0.35 0.49>
     outlineSettings {
      bUseBrushTransparency=false
      roundingType=HalfHeightRadius
      width=0
      color {colorUseRule=UseColor_Specified, specifiedColor=<rgba 0 0 0 0>}
      cornerRadii=<0 0 0 0>
     }
     tintColor {colorUseRule=UseColor_Specified, specifiedColor=<rgba 1 1 1 1>}
     uVRegion {bIsValid=false, max=<0 0>, min=<0 0>}
    }
    renderTransformPivot=<0.5 0.9>
   }
   slot {class=WidgetSwitcherSlot, properties {padding=<ltrb 0 -3 0 -34>}}
  }
  Pop {
   widgetClassPath=&Image
   bIsVariable=true
   parent=Switcher_Styles
   widget=Pop
   properties {
    visibility=SelfHitTestInvisible
    brush {
     bIsDynamicallyLoaded=false
     drawAs=Box
     imageType=FullColor
     mirroring=NoMirror
     resourceName=None
     resourceObject=/CreativeCommunication/BubbleChat/Art/Textures/T_UI_BubbleChatStyle_Pop
     tiling=NoTile
     imageSize=<32 32>
     margin=<ltrb 0.49 0.17 0.49 0.45>
     outlineSettings {
      bUseBrushTransparency=false
      roundingType=HalfHeightRadius
      width=0
      color {colorUseRule=UseColor_Specified, specifiedColor=<rgba 0 0 0 0>}
      cornerRadii=<0 0 0 0>
     }
     tintColor {colorUseRule=UseColor_Specified, specifiedColor=<rgba 1 1 1 1>}
     uVRegion {bIsValid=false, max=<0 0>, min=<0 0>}
    }
    renderTransformPivot=<0.5 0.9>
   }
   slot {class=WidgetSwitcherSlot, properties {padding=<ltrb -15 -16 -26 -58>}}
  }
  Retro {
   widgetClassPath=&Image
   bIsVariable=true
   parent=Switcher_Styles
   widget=Retro
   properties {
    visibility=SelfHitTestInvisible
    brush {
     bIsDynamicallyLoaded=false
     drawAs=Box
     imageType=FullColor
     mirroring=NoMirror
     resourceName=None
     resourceObject=/CreativeCommunication/BubbleChat/Art/Textures/T_UI_BubbleChatStyle_Retro
     tiling=NoTile
     imageSize=<94 94>
     margin=<ltrb 0.2526 0.1 0.339 0.462>
     outlineSettings {
      bUseBrushTransparency=false
      roundingType=HalfHeightRadius
      width=0
      color {colorUseRule=UseColor_Specified, specifiedColor=<rgba 0 0 0 0>}
      cornerRadii=<0 0 0 0>
     }
     tintColor {colorUseRule=UseColor_Specified, specifiedColor=<rgba 1 1 1 1>}
     uVRegion {bIsValid=false, max=<0 0>, min=<0 0>}
    }
    renderTransformPivot=<0.5 0.9>
   }
   slot {class=WidgetSwitcherSlot, properties {padding=<ltrb -5 -2 -5 -42>}}
  }
  Rounded {
   widgetClassPath=&Image
   bIsVariable=true
   parent=Switcher_Styles
   widget=Rounded
   properties {
    visibility=SelfHitTestInvisible
    brush {
     bIsDynamicallyLoaded=false
     drawAs=Box
     imageType=FullColor
     mirroring=NoMirror
     resourceName=None
     resourceObject=/CreativeCommunication/BubbleChat/Art/MI_UI_Creative_SpeechBubble_OverheadDialogue
     tiling=NoTile
     imageSize=<94 94>
     margin=<ltrb 0.49 0.49 0.49 0.49>
     outlineSettings {
      bUseBrushTransparency=false
      roundingType=HalfHeightRadius
      width=0
      color {colorUseRule=UseColor_Specified, specifiedColor=<rgba 0 0 0 0>}
      cornerRadii=<0 0 0 0>
     }
     tintColor {colorUseRule=UseColor_Specified, specifiedColor=<rgba 1 1 1 1>}
     uVRegion {bIsValid=false, max=<0 0>, min=<0 0>}
    }
    colorAndOpacity=<rgba 0.0395 0.0578 0.4072 1>
    renderTransformPivot=<0.5 0.9>
   }
   slot {class=WidgetSwitcherSlot, properties {padding=<ltrb -23 -23 -23 -24>}}
  }
  ColorSwitchingCB {
   widgetClassPath=&CommonBorder
   bIsVariable=true
   parent=Overlay_BubbleContainer
   widget=ColorSwitchingCB
   properties {contentColorAndOpacity=<rgba 0 0 0 1>, padding=<ltrb 0 0 0 0>}
   slot {class=OverlaySlot, properties {horizontalAlignment=HAlign_Fill, verticalAlignment=VAlign_Fill, padding=<ltrb 18 10 18 10>}}
  }
  ListView_Messages {
   widgetClassPath=&CommonListView
   bIsVariable=true
   parent=ColorSwitchingCB
   widget=ListView_Messages
   properties {
    allowOverscroll=false
    bAllowDragging=false
    bEnableRightClickScrolling=false
    bEnableScrollAnimation=false
    bEnableTouchScrolling=false
    bIsFocusable=false
    clipping=Inherit
    consumeMouseWheel=Never
    entryWidgetClass=&/WBP_CreativeBubbleChatEntry.WBP_CreativeBubbleChatEntry_C
    navigation=&self:WidgetTree.ListView_Messages.WidgetNavigation_0
    numDesignerPreviewEntries=1
    selectionMode=None
    visibility=SelfHitTestInvisible
    wheelScrollMultiplier=0
    scrollBarStyle {
     thickness=0
     draggedThumbImage {
      bIsDynamicallyLoaded=false
      drawAs=Box
      imageType=FullColor
      mirroring=NoMirror
      resourceName=../../../Engine/Content/Slate/Common/Scrollbar_Thumb.png
      resourceObject=null
      tiling=NoTile
      imageSize=<0 0>
      margin=<ltrb 0.25 0.25 0.25 0.25>
      outlineSettings {
       bUseBrushTransparency=false
       roundingType=HalfHeightRadius
       width=0
       color {colorUseRule=UseColor_Specified, specifiedColor=<rgba 0 0 0 0>}
       cornerRadii=<0 0 0 0>
      }
      tintColor {colorUseRule=UseColor_Specified, specifiedColor=<rgba 1 1 1 1>}
      uVRegion {bIsValid=false, max=<0 0>, min=<0 0>}
     }
     horizontalBackgroundImage {
      bIsDynamicallyLoaded=false
      drawAs=NoDrawType
      imageType=NoImage
      mirroring=NoMirror
      resourceName=None
      resourceObject=null
      tiling=NoTile
      imageSize=<0 0>
      margin=<ltrb 0 0 0 0>
      outlineSettings {
       bUseBrushTransparency=false
       roundingType=HalfHeightRadius
       width=0
       color {colorUseRule=UseColor_Specified, specifiedColor=<rgba 0 0 0 0>}
       cornerRadii=<0 0 0 0>
      }
      tintColor {colorUseRule=UseColor_Specified, specifiedColor=<rgba 1 1 1 1>}
      uVRegion {bIsValid=false, max=<0 0>, min=<0 0>}
     }
     horizontalBottomSlotImage {
      bIsDynamicallyLoaded=false
      drawAs=Image
      imageType=FullColor
      mirroring=NoMirror
      resourceName=../../../Engine/Content/Slate/Common/Scrollbar_Background_Horizontal.png
      resourceObject=null
      tiling=NoTile
      imageSize=<8 8>
      margin=<ltrb 0 0 0 0>
      outlineSettings {
       bUseBrushTransparency=false
       roundingType=HalfHeightRadius
       width=0
       color {colorUseRule=UseColor_Specified, specifiedColor=<rgba 0 0 0 0>}
       cornerRadii=<0 0 0 0>
      }
      tintColor {colorUseRule=UseColor_Specified, specifiedColor=<rgba 1 1 1 1>}
      uVRegion {bIsValid=false, max=<0 0>, min=<0 0>}
     }
     horizontalTopSlotImage {
      bIsDynamicallyLoaded=false
      drawAs=Image
      imageType=FullColor
      mirroring=NoMirror
      resourceName=../../../Engine/Content/Slate/Common/Scrollbar_Background_Horizontal.png
      resourceObject=null
      tiling=NoTile
      imageSize=<8 8>
      margin=<ltrb 0 0 0 0>
      outlineSettings {
       bUseBrushTransparency=false
       roundingType=HalfHeightRadius
       width=0
       color {colorUseRule=UseColor_Specified, specifiedColor=<rgba 0 0 0 0>}
       cornerRadii=<0 0 0 0>
      }
      tintColor {colorUseRule=UseColor_Specified, specifiedColor=<rgba 1 1 1 1>}
      uVRegion {bIsValid=false, max=<0 0>, min=<0 0>}
     }
     hoveredThumbImage {
      bIsDynamicallyLoaded=false
      drawAs=Box
      imageType=FullColor
      mirroring=NoMirror
      resourceName=../../../Engine/Content/Slate/Common/Scrollbar_Thumb.png
      resourceObject=null
      tiling=NoTile
      imageSize=<0 0>
      margin=<ltrb 0.25 0.25 0.25 0.25>
      outlineSettings {
       bUseBrushTransparency=false
       roundingType=HalfHeightRadius
       width=0
       color {colorUseRule=UseColor_Specified, specifiedColor=<rgba 0 0 0 0>}
       cornerRadii=<0 0 0 0>
      }
      tintColor {colorUseRule=UseColor_Specified, specifiedColor=<rgba 1 1 1 1>}
      uVRegion {bIsValid=false, max=<0 0>, min=<0 0>}
     }
     normalThumbImage {
      bIsDynamicallyLoaded=false
      drawAs=Box
      imageType=FullColor
      mirroring=NoMirror
      resourceName=../../../Engine/Content/Slate/Common/Scrollbar_Thumb.png
      resourceObject=null
      tiling=NoTile
      imageSize=<0 0>
      margin=<ltrb 0.25 0.25 0.25 0.25>
      outlineSettings {
       bUseBrushTransparency=false
       roundingType=HalfHeightRadius
       width=0
       color {colorUseRule=UseColor_Specified, specifiedColor=<rgba 0 0 0 0>}
       cornerRadii=<0 0 0 0>
      }
      tintColor {colorUseRule=UseColor_Specified, specifiedColor=<rgba 1 1 1 1>}
      uVRegion {bIsValid=false, max=<0 0>, min=<0 0>}
     }
     verticalBackgroundImage {
      bIsDynamicallyLoaded=false
      drawAs=NoDrawType
      imageType=NoImage
      mirroring=NoMirror
      resourceName=None
      resourceObject=null
      tiling=NoTile
      imageSize=<0 0>
      margin=<ltrb 0 0 0 0>
      outlineSettings {
       bUseBrushTransparency=false
       roundingType=HalfHeightRadius
       width=0
       color {colorUseRule=UseColor_Specified, specifiedColor=<rgba 0 0 0 0>}
       cornerRadii=<0 0 0 0>
      }
      tintColor {colorUseRule=UseColor_Specified, specifiedColor=<rgba 1 1 1 1>}
      uVRegion {bIsValid=false, max=<0 0>, min=<0 0>}
     }
     verticalBottomSlotImage {
      bIsDynamicallyLoaded=false
      drawAs=Image
      imageType=FullColor
      mirroring=NoMirror
      resourceName=../../../Engine/Content/Slate/Common/Scrollbar_Background_Vertical.png
      resourceObject=null
      tiling=NoTile
      imageSize=<8 8>
      margin=<ltrb 0 0 0 0>
      outlineSettings {
       bUseBrushTransparency=false
       roundingType=HalfHeightRadius
       width=0
       color {colorUseRule=UseColor_Specified, specifiedColor=<rgba 0 0 0 0>}
       cornerRadii=<0 0 0 0>
      }
      tintColor {colorUseRule=UseColor_Specified, specifiedColor=<rgba 1 1 1 1>}
      uVRegion {bIsValid=false, max=<0 0>, min=<0 0>}
     }
     verticalTopSlotImage {
      bIsDynamicallyLoaded=false
      drawAs=Image
      imageType=FullColor
      mirroring=NoMirror
      resourceName=../../../Engine/Content/Slate/Common/Scrollbar_Background_Vertical.png
      resourceObject=null
      tiling=NoTile
      imageSize=<8 8>
      margin=<ltrb 0 0 0 0>
      outlineSettings {
       bUseBrushTransparency=false
       roundingType=HalfHeightRadius
       width=0
       color {colorUseRule=UseColor_Specified, specifiedColor=<rgba 0 0 0 0>}
       cornerRadii=<0 0 0 0>
      }
      tintColor {colorUseRule=UseColor_Specified, specifiedColor=<rgba 1 1 1 1>}
      uVRegion {bIsValid=false, max=<0 0>, min=<0 0>}
     }
    }
   }
   slot {class=BorderSlot, properties {padding=<ltrb 0 0 0 0>}}
  }
 ]
}
