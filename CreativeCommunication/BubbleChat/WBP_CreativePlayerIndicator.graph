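# Creative player indicator widget that anchors a chat bubble for a targeted player, inside a HUD wrapper tagged hide-for-capture.
# A target change offsets the anchor for local versus other players and creates the bubble chat viewmodel.
# Settings updates switch built-in styles or async-load a custom widget class, then inject the viewmodel through MVVM.
# Visibility events toggle the bubble, while focus changes drive a timer-delayed detail animation that has no bindings.
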WidgetBlueprint /CreativeCommunication/BubbleChat/WBP_CreativePlayerIndicator
animations [Anim_FadeIn {
  binding_count=0
  length_seconds=0.15
  object=&self:Anim_FadeIn
  sequence {bindings [], display_rate {numerator=60}}
 }]
graphs [
 EventGraph {nodes [
   OnRecentlyUpdated() {is=title K2Node_CustomEvent}
   > DelayUntilNextTick!(execute) -> (then)
   > Set bRecentlyUpdated(bRecentlyUpdated:bool=false) -> (Output_Get:bool)
   n1 = Event OnBubbleChatVisibilityChanged!() -> (OutputDelegate, then, bNewIsVisible:bool) {is=title K2Node_Event}
   > Set InitBubbleChatVisibility(InitBubbleChatVisibility:bool=@n1.bNewIsVisible) -> (Output_Get:bool)
   > SetVisibility(
    self=Get CurrentCreativeBubbleChat() -> (CurrentCreativeBubbleChat:User Widget Object Reference)
    InVisibility=Select(Option 0=Collapsed, Option 1=Visible, Index=@n1.bNewIsVisible)
   )
   n2 = Event OnInFocusStateChanged!() -> (OutputDelegate, then, bNewIsInFocus:bool) {is=title K2Node_Event}
   > ShowDetails!(execute, self, Condition:bool=@n2.bNewIsInFocus) -> (then)
   Event Construct()
   > SetMobileScale()
   n3 = Get Anim_FadeIn() -> (Anim_FadeIn:Widget Animation Object Reference)
   > PlayAnimation(InAnimation=@n3, StartAtTime=GetStartTime!(self:Widget Animation Object Reference=@n3) -> (ReturnValue:float), PlayMode=Reverse)
   > n4 = Set CurrentCreativeBubbleChat(CurrentCreativeBubbleChat:User Widget Object Reference=Get WBP_CreativeBubbleChat() -> (WBP_CreativeBubbleChat:WBP Creative Bubble Chat V2 Object Reference)) -> (Output_Get:User Widget Object Reference)
   > SetVisibility(self=@n4, InVisibility=Collapsed)
   n5 = Event OnCreativeBubbleChatSettingsUpdated_BP!() -> (OutputDelegate, then, BubbleChatStyle:EBubbleChatStyle Enum=Default, CustomBubbleWidgetClass:"User Widget Soft Class Reference (by ref)") {is=title K2Node_Event}
   > n6 = Branch(Condition="Equal (Enum)"(A:EBubbleChatStyle Enum=@n5.BubbleChatStyle, B:EBubbleChatStyle Enum=Custom))
   > n7 = Branch(Condition="Not Equal (SoftClassReference)"(
     A:"Object Soft Class Reference (by ref)"=Get LastCustomBubbleChatWidgetClass() -> (LastCustomBubbleChatWidgetClass:User Widget Soft Class Reference)
     B:"Object Soft Class Reference (by ref)"=@n5.CustomBubbleWidgetClass
    ) -> (ReturnValue:bool=~) {is=title K2Node_PromotableOperator})
   > Set LastCustomBubbleChatWidgetClass(LastCustomBubbleChatWidgetClass:User Widget Soft Class Reference=@n5.CustomBubbleWidgetClass) -> (Output_Get:User Widget Soft Class Reference)
   > n8 = CreateWidgetAsync!(
    execute:Exec
    UserWidgetSoftClass:User Widget Soft Class Reference=@n5.CustomBubbleWidgetClass
    OwningPlayer:Player Controller Object Reference=GetOwningPlayer()
    bSuspendInputUntilComplete:bool=false
   ) -> (then:Exec, OnComplete:Exec, OnFailure:Exec, UserWidget:User Widget Object Reference) {is=title K2Node_AsyncAction}
   SetBubbleChatViewModelOnUserWidget!(execute=@n8.OnComplete, self, UserWidget:User Widget Object Reference=@n8) -> (then)
   SetBubbleChatViewModelOnUserWidget(execute=@n7.else, UserWidget=Get CurrentCreativeBubbleChat())
   Set ActiveWidgetIndex!(
    execute=@n6.else
    ActiveWidgetIndex:int="To Integer (Byte)"!(InByte:Byte=@n5.BubbleChatStyle) -> (ReturnValue:int=0)
    self:Widget Switcher Object Reference=Get Switcher_Styles!(self:WBP Creative Bubble Chat V2 Object Reference=Get WBP_CreativeBubbleChat()) -> (Switcher_Styles:Widget Switcher Object Reference)
   ) -> (then, Output_Get:int=0)
   n9 = Get WBP_CreativeBubbleChat()
   > SetVisualStyle!(execute, self:WBP Creative Bubble Chat V2 Object Reference=@n9, IsWhisper:bool=false) -> (then)
   > SetBubbleChatViewModelOnUserWidget(UserWidget=@n9)
   n10 = Event OnTargetPlayerChanged!() -> (OutputDelegate, then, TargetedPlayer:Fort Player State Athena Object Reference) {is=title K2Node_Event}
   > Is Valid(InputObject=@n10.TargetedPlayer)
   > Set CachedTargetedPlayer(CachedTargetedPlayer:Fort Player State Athena Object Reference=@n10.TargetedPlayer) -> (Output_Get:Fort Player State Athena Object Reference)
   n11 = GetOwningPlayer()
   > Is Valid(InputObject=@n11)
   > n12 = Cast To FortPlayerStateAthena!(execute, Object=Get PlayerState!(self:Controller Object Reference=@n11) -> (PlayerState:Player State Object Reference)) -> (then, CastFailed, AsFort Player State Athena:Fort Player State Athena Object Reference) {is=title K2Node_DynamicCast}
   > Set bIsLocalPlayer(bIsLocalPlayer:bool="Equal (Object)"(A=@n12, B=Get CachedTargetedPlayer() -> (CachedTargetedPlayer:Fort Player State Athena Object Reference))) -> (Output_Get:bool)
   > n13 = Sequence()
   n14 = Get bIsLocalPlayer() -> (bIsLocalPlayer:bool)
   > SetRenderTranslation(
    self=Get ChatAnchor() -> (ChatAnchor:Fort Visual Attachment Object Reference)
    Translation=MakeVector2D(X=0, Y=Select(Option 0:double=-8, Option 1:double=0, Index=@n14) -> (ReturnValue:double))
   )
   n15 = Construct Object from Class!(execute=@n13.then_1, Class=/Script/CreativeCommunicationUI.CreativeBubbleChatViewModel, self, Message:Text, bIsIndicatorClamped:bool=False) -> (then, ReturnValue:Bubble Chat Viewmodel Object Reference)
   > SetupBubbleChatViewModel!(execute, self:Creative Player Indicator Widget Object Reference, ViewModel:Bubble Chat Viewmodel Object Reference=@n15) -> (then)
  ]}
 HideDetails {nodes [
   HideDetails() {is=title K2Node_FunctionEntry}
   > Branch(Condition=Get bShowingDetails() -> (bShowingDetails:bool))
   > PlayAnimationReverse(InAnimation=Get Anim_FadeIn())
   > Set bShowingDetails(bShowingDetails:bool=false) -> (Output_Get:bool)
   > Clear and Invalidate Timer by Handle(Handle=Get UpdateTimer() -> (UpdateTimer:Timer Handle Structure))
  ]}
 SetBubbleChatViewModelOnUserWidget {nodes [
   n1 = SetBubbleChatViewModelOnUserWidget!() -> (then, UserWidget:User Widget Object Reference) {is=title K2Node_FunctionEntry}
   > RemoveFromParent(self=Get CurrentCreativeBubbleChat())
   n2 = Get View From User Widget!(
    self:Viewmodel Engine Subsytem Object Reference=Get MVVMSubsystem() -> (ReturnValue:Viewmodel Engine Subsytem Object Reference) {is=title K2Node_GetEngineSubsystem}
    UserWidget:User Widget Object Reference=@n1
   ) -> (ReturnValue:MVVM View Object Reference) {is=title K2Node_CallFunction}
   > Is Valid(InputObject=@n2)
   > SetViewModel!(
    execute
    self:MVVM View Object Reference=@n2
    ViewModelName:Name=CreativeBubbleChatViewModel
    ViewModel:Notify Field Value Changed Interface=GetBubbleChatViewModel!(self:Creative Player Indicator Widget Object Reference) -> (ReturnValue:Bubble Chat Viewmodel Object Reference)
   ) -> (then, ReturnValue:bool)
   > AddChild(self=Get ChatAnchor(), Content=Get UserWidget() -> (UserWidget:User Widget Object Reference))
   > n3 = Set CurrentCreativeBubbleChat(CurrentCreativeBubbleChat=Get UserWidget())
   > SetVisibility(
    self=@n3
    InVisibility=Select(
     Option 0:ESlateVisibility Enum=Collapsed
     Option 1:ESlateVisibility Enum=Visible
     Index=Get InitBubbleChatVisibility() -> (InitBubbleChatVisibility:bool)
    ) -> (ReturnValue:ESlateVisibility Enum)
   )
  ]}
 SetMobileScale {nodes [
   SetMobileScale() {is=title K2Node_FunctionEntry}
   > Branch(Condition=IsMobileGame())
   > SetRenderScale!(execute, self:Widget Object Reference=Get PrimaryOvr() -> (PrimaryOvr:Overlay Object Reference), Scale:Vector 2D Structure=(X=1.15,Y=1.15)) -> (then)
  ]}
 ShowDetails {nodes [
   n1 = ShowDetails!() -> (then, Condition:bool=~) {is=title K2Node_FunctionEntry}
   > Clear and Invalidate Timer by Handle(Handle=Get UpdateTimer())
   > n2 = Branch(Condition=@n1)
   > n3 = Branch(Condition=Get bShowingDetails())
   n4 = Get Anim_FadeIn()
   StopAnimation(execute=@n3.else, InAnimation=@n4)
   > PlayAnimationForward(InAnimation=@n4)
   > Set bShowingDetails(bShowingDetails=true)
   n5 = Set Timer by Event(execute=@n2.else, Delegate=Create Event(), Time=Get HideTimer() -> (HideTimer:float), bLooping=false)
   > Set UpdateTimer(UpdateTimer:Timer Handle Structure=@n5) -> (Output_Get:Timer Handle Structure)
  ]}
]
variables [
 bIsLocalPlayer {type=boolean}
 bRecentlyUpdated {type=boolean}
 bShowingDetails {type=boolean}
 CachedTargetedPlayer {type=FortPlayerStateAthena, default=null}
 CurrentCreativeBubbleChat {type=UserWidget, default=null}
 HideTimer {type=number, default=0.2}
 InitBubbleChatVisibility {type=boolean}
 LastCustomBubbleChatWidgetClass {type=Class, default=null}
 UpdateTimer {type=TimerHandle}
]
widgets {
 info {
  inheritedWidgetCount=2
  namedSlotCount=1
  parentClass=/Script/CreativeCommunicationUI.CreativePlayerIndicatorWidget
  rootWidgetClass=&Overlay
  widgetCount=7
 }
 widgets [
  PrimaryOvr {
   widgetClassPath=&Overlay
   bIsVariable=true
   widget=PrimaryOvr
   properties {renderTransform {angle=0, scale=<1 1>, shear=<0 0>, translation=<0 -80>}}
  }
  CollapsedWidgets {
   widgetClassPath=&Overlay
   parent=PrimaryOvr
   widget=CollapsedWidgets
   properties {visibility=Collapsed}
   slot {class=OverlaySlot, properties {horizontalAlignment=HAlign_Center, verticalAlignment=VAlign_Center}}
  }
  Text_Distance {
   widgetClassPath=&CommonTextBlock
   bInherited=true
   bIsVariable=true
   parent=CollapsedWidgets
   widget=Text_Distance
   properties {
    justification=Center
    text=99m
    visibility=Collapsed
    colorAndOpacity {colorUseRule=UseColor_Specified, specifiedColor=<rgba 1 1 1 0.75>}
    font {
     bForceMonospaced=false
     bMaterialIsStencil=false
     fontMaterial=/Game/Athena/HUD/Team/M_Font_PlayerIndicator
     fontObject=&HeadingNow
     letterSpacing=0
     monospacedWidth=1
     size=12
     skewAmount=0
     typefaceFontName=75 Medium Italic
     outlineSettings {
      bApplyOutlineToDropShadows=false
      bMiteredCorners=false
      bSeparateFillAlpha=false
      outlineMaterial=null
      outlineSize=0
      outlineColor=<rgba 0 0 0 1>
     }
    }
   }
   slot {class=OverlaySlot}
  }
  FortHUDElementWrapper_Chat {
   widgetClassPath=/Game/Athena/HUD/FortHUDElementWrapper
   bIsVariable=true
   parent=PrimaryOvr
   widget=FortHUDElementWrapper_Chat
   properties {hiddenVisibility=HitTestInvisible, elementTag {tagName=Athena.HideForCapture}, hUDElementTag {gameplayTags []}}
   slot {class=OverlaySlot}
  }
  ChatAnchor {
   widgetClassPath=&FortVisualAttachment
   bIsVariable=true
   namedSlotHost=FortHUDElementWrapper_Chat
   widget=ChatAnchor
   properties {contentAnchor=<0.5 1>, renderTransform {angle=0, scale=<1 1>, shear=<0 0>, translation=<0 -8>}}
   slot=~
  }
  WBP_CreativeBubbleChat {
   widgetClassPath=&/WBP_CreativeBubbleChat_V2
   bIsVariable=true
   parent=ChatAnchor
   widget=WBP_CreativeBubbleChat
   slot {class=SizeBoxSlot}
  }
  TutorialHighlight {widgetClassPath=&TutorialHighlight}
 ]
}
